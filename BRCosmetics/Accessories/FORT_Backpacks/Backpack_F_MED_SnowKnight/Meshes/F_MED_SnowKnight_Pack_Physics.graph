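# Physics asset for the SnowKnight backpack cosmetic mesh, giving the pack simulated secondary motion.
# It has two capsule bodies, a small one on the pack bone and a larger, offset one on dyn_backpack.
# A single constraint links dyn_backpack to pack with tight swing and twist limits, so the dynamic part can only sway slightly.
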PhysicsAsset /BRCosmetics/Accessories/FORT_Backpacks/Backpack_F_MED_SnowKnight/Meshes/F_MED_SnowKnight_Pack_Physics
bodies {
 dyn_backpack [{shapeType=Capsule, length=2, radius=1, shapeName=dyn_backpack_capsule, center=<1.5 0 0>, rotation=<rot 90 0 0>}]
 pack [{shapeType=Capsule, length=0.505, radius=0.505, shapeName=pack_capsule, rotation=<rot 0 -90 0>}]
}
constraints [{bone1Name=dyn_backpack, bone2Name=pack, swing1LimitDegrees=5, swing2LimitDegrees=10, twistLimitDegrees=1}]
