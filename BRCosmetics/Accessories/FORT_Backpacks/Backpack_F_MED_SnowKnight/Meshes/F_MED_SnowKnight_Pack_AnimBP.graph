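# Animation Blueprint for the SnowKnight backpack cosmetic, adding physics and spine-reactive offsets to the pack mesh.
# The AnimGraph nudges the pack bone, runs RigidBody simulation in root-bone space outside the front end, then additively moves dyn_backpack.
# The EventGraph maps spine_05-versus-spine_01 yaw to that offset and rotation.
# It also computes sine-wave float and movement-state ambient values, which the AnimGraph leaves unused.
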AnimBlueprint /BRCosmetics/Accessories/FORT_Backpacks/Backpack_F_MED_SnowKnight/Meshes/F_MED_SnowKnight_Pack_AnimBP
target_skeleton=&/F_MED_SnowKnight_Pack_Skeleton
graphs [
 AnimGraph {nodes [
   n1 = "Transform (Modify) Bone - Bone: pack"(ComponentPose=Local To Component(LocalPose=Input Pose()), Translation="0,-2.5,0", Rotation="0,-2.5,0") {
    rotationMode=BMM_Additive
    rotationSpace=BCS_BoneSpace
    translationMode=BMM_Additive
    translationSpace=BCS_BoneSpace
    alphaBoolBlend {blendOption=Linear}
    boneToModify {boneName=pack}
    rotation {yaw=-2.5}
    translation {y=-2.5}
   }
   n2 = RigidBody!(ComponentPose=@n1, bAlphaBoolEnabled:bool=NOT Boolean(A=Get bIsInFrontEnd()), OverrideWorldGravity:Vector="0,0,-300", ExternalForce) -> (Pose) {
    alphaInputType=Bool
    bOverrideWorldGravity=true
    overridePhysicsAsset=&/F_MED_SnowKnight_Pack_Physics
    simulationSpace=BaseBoneSpace
    alphaBoolBlend {blendOption=Linear}
    baseBoneRef {boneName=root}
    componentAppliedLinearAccClamp=<3000 3000 3000>
    componentLinearAccScale=<0.35 0.35 0.35>
    componentLinearVelScale=<0.35 0.35 0.35>
    simSpaceSettings {
     maxAngularAcceleration=1000
     maxAngularVelocity=3
     maxLinearAcceleration=700
     maxLinearVelocity=500
     velocityScaleZ=0
     externalLinearDragV=<0.2 0.2>
    }
   }
   n3 = "Transform (Modify) Bone - Bone: dyn_backpack"(ComponentPose=@n2, Alpha=1, Translation=Get backpack_translation(), Rotation=Get backpack_rotate(), Scale="1,1,1") {rotationMode=BMM_Additive, translationMode=BMM_Additive, alphaBoolBlend {blendOption=Linear}, boneToModify {boneName=dyn_backpack}}
   Output Pose(Result=Component To Local(ComponentPose=@n3)) {layerGroup=None}
  ]}
 EventGraph {nodes [
   Event BlueprintUpdateAnimation()
   > Set IsSkyDive(IsSkyDive:float="To Float (Boolean)"(InBool=Get bIsSkyDiving())) -> (Output_Get:float)
   n1 = GetOwnerSkeletalMeshComponent()
   n2 = BreakRotator(InRot=BreakTransform(InTransform=GetDeltaTransformFromRefPose(self=@n1, BoneName=spine_05, BaseName=spine_01)).Rotation)
   > Set backpack_rotate(backpack_rotate=MakeRotator(Roll=MapRangeClamped(Value=@n2.Yaw, InRangeA=0, InRangeB=-20, OutRangeB=20)))
   > Set backpack_translation(backpack_translation=MakeVector(Y=MapRangeClamped(Value=@n2.Yaw, InRangeA=0, InRangeB=-15, OutRangeA=-0.5, OutRangeB=-2)))
   n3 = GetRealTimeSeconds()
   n4 = "Sin (Radians)"(A=float * float(A=float * float(A=1.1, B=1.1), B=@n3))
   n5 = "Sin (Radians)"(A=float * float(A=float * float(A=2, B=1.3), B=@n3))
   n6 = "Sin (Radians)"(A=float * float(A=float * float(A=3, B=1.3), B=@n3))
   > Set ambient_Float(ambient_Float:Vector=MakeVector(
     X=Remap Float(Input=@n6, OldRange=(X=-1,Y=1), NewRange=(X=-1.2,Y=0.5)) -> (Output=~)
     Y=float * float(A=@n4, B=0.3)
     Z=float * float(A=@n5, B=0.6)
    )) -> (Output_Get:Vector)
   > n7 = Branch(Condition=OR Boolean!(A=Get bIsJumping(), B=Get bIsFalling(), C:bool=Get bIsTacticalSprinting(), D:bool=Get bIsPlayingMeleeAnim()) -> (ReturnValue))
   > Set ambient motion(ambient motion:float=1) -> (Output_Get:float)
   n8 = Branch(execute=@n7.else, Condition=OR Boolean!(A=Get bIsSprinting(), B=Get bIsMoving2D(), C=Get bIsSkydiveFloating()) -> (ReturnValue))
   > Set ambient motion(ambient motion=0.7)
   Set ambient motion(execute=@n8.else, ambient motion=0)
   TryGetPawnOwner()
  ]}
]
variables [
 ambient motion {type=number, default=1}
 ambient_Float {type=Vector}
 backpack_rotate {type=Rotator}
 backpack_translation {type=Vector}
 IsSkyDive {type=number, default=1}
]
